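# Niagara system for the PineTrim_Crisp emote's "200IQ" effect: short-lived red body lights drawn as sprites.
# One emitter burst-spawns particles, fades them in and out by curve, and draws them through two sprite renderers.
# A separate Linear Color attribute pulses through a waveform whose input is offset per spawn group, staggering the blinking.
# An indoor-player parameter lerps both colors to dimmer values.
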NiagaraSystem /BRCosmetics/Animation/Game/MainPlayer/Emotes/PineTrim_Crisp/FX/NS_Emote_PineTrim_Crisp_200IQ
dependencies {usedDataInterfaces [&NiagaraDataInterfaceCurve], usedDynamicInputs [], usedModules [&SystemState]}
emitters {BodyLights001 {
  emitterUpdateScript [
   &EmitterState(Life Cycle Mode=::NewEnumerator1, Loop Behavior=::NewEnumerator1, UseLoopDelay=false)
   &SpawnBurst_Instantaneous(Use Spawn Probability=false, UseLoopCountLimit=false)
  ]
  particleSpawnScript [
   &InitializeParticle(
    Color=&LerpLinearColors(StartColor=<rgba 15 3.194 3.0469 2>, EndColor=<rgba 1 0.7125 0.7089 0.6>, LerpFactor=$NPC.FortniteNPC.IsPlayerIndoors)
    Lifetime=3.2
    Mass Max=1.5
    Mass Mode=::NewEnumerator3
    Sprite Size Mode=::NewEnumerator3
    Uniform Sprite Size=48.9235
    UsePositionOffset=true
    Write Material Random=true
   )
   &CameraOffset(Camera Offset Amount=1)
  ]
  particleUpdateScript [
   &ParticleState()
   &ScaleColor(
    Scale Alpha=&FloatFromCurve(FloatCurve=DataInterface() {propertyValues {Curve {keys [<at -0.0017 0>,<at 0.0725 1.0029>,<at 0.7283 1>,<at 1 0>]}}})
    ScaleA=true
    ScaleRGB=true
   )
   /SetVariables(Particles.Linear Color=&Multiply_LinearColorByFloat(
     Linear Color=&LerpLinearColors(StartColor=<rgba 1 0.0123 0 3>, EndColor=<rgba 0.6707 0.0083 0 0.25>, LerpFactor=$NPC.FortniteNPC.IsPlayerIndoors)
     Float=&Multiply_Float(
      A=&Waveform(
       Global Input Position=&MergeHelper_FloatSwitch(0=0.35, 1=0.25, 2=0.15, 3=0.05, Index=&SpawnGroup(), Num=::NewEnumerator2)
       "Waveform [1]"=::NewEnumerator0
       "[1] Input Position"=-0.7296
       "[1] Frequency"=1
       "[1] Amplitude Min/Max"=<0 1>
       "[1] Bias"=1
      )
      B=&FloatFromCurve(FloatCurve=DataInterface() {propertyValues {Curve {keys [<at -0.0017 0>,<at 0.0725 1.0029>,<at 0.7283 1>,<at 1 0>]}}}, Scale Curve=2.9)
     )
    )) {isSetParameters=true}
  ]
  rendererClasses [&NiagaraSpriteRendererProperties]
  renderers [
   {rendererClass=&NiagaraSpriteRendererProperties}
   {rendererClass=&NiagaraSpriteRendererProperties, rendererIndex=1}
  ]
 }}
summary {
 systemName=NS_Emote_PineTrim_Crisp_200IQ
 emitters [{emitterName=BodyLights001, rendererClasses [&NiagaraSpriteRendererProperties]}]
 userVariables [User.BlinkFreq {type=NiagaraFloat, defaultValue=1, description=""}]
}
