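# Niagara light effect for the PineTrim_Crisp emote, showing blinking warm-colored lights on body and wings over 13 seconds.
# BodyLights places four offset, tinted sprites, WingLights pins two sprites to skeletal mesh bones or sockets, and MainLight combines light, mesh and sprite renderers.
# User.BlinkFreq drives alpha waveforms, with age curves changing the pattern and brightness around 8.4 seconds.
# Main and wing emitters add a short fading flash burst at 8.7 seconds.
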NiagaraSystem /BRCosmetics/Animation/Game/MainPlayer/Emotes/PineTrim_Crisp/FX/NS_Emote_PineTrim_Crisp_LightFX
dependencies {
 usedDataInterfaces [
  &NiagaraDataInterfaceCurve
  &NiagaraDataInterfaceArrayInt32
  &NiagaraDataInterfaceArrayFloat3
  &NiagaraDataInterfaceSkeletalMesh
 ]
 usedDynamicInputs []
 usedModules [
  &SystemState
  &self:SystemSpawnScript.SystemScriptSource.SystemScriptGraph.NiagaraNodeAssignment_0.SetVariables_#1
 ]
}
emitters {
 BodyLights {
  emitterUpdateScript [
   &EmitterState(Life Cycle Mode=::NewEnumerator1, Loop Behavior=::NewEnumerator1, UseLoopDelay=false)
   &SpawnBurst_Instantaneous(Use Spawn Probability=false, UseLoopCountLimit=false)
   &SpawnBurst_Instantaneous(Spawn Group=1, Use Spawn Probability=false, UseLoopCountLimit=false) {module=SpawnBurst_Instantaneous001}
   &SpawnBurst_Instantaneous(Spawn Group=2, Use Spawn Probability=false, UseLoopCountLimit=false) {module=SpawnBurst_Instantaneous002}
   &SpawnBurst_Instantaneous(Spawn Group=3, Use Spawn Probability=false, UseLoopCountLimit=false) {module=SpawnBurst_Instantaneous003}
  ]
  particleSpawnScript [
   /SetVariables(Particles.VisibilityTag=&SpawnGroup()) {isSetParameters=true}
   &InitializeParticle(
    Color=&MergeHelper_ColorSwitch(
     0=&FNTimeOfDayColor(
      Night Color=<rgba 1 0.8674 0.4583 0.6>
      Morning Color=<rgba 1 0.8674 0.4583 0.7>
      Day Color=<rgba 1 0.8674 0.4583 0.7>
      Evening Color=<rgba 1 0.8674 0.4583 0.7>
     )
     1=&FNTimeOfDayColor(
      Night Color=<rgba 1 0.5181 0.2292 0.7>
      Morning Color=<rgba 1 0.5181 0.2292 0.7>
      Day Color=<rgba 1 0.5181 0.2292 0.7>
      Evening Color=<rgba 1 0.5181 0.2292 0.7>
     )
     2=&FNTimeOfDayColor(
      Night Color=<rgba 1 0.2587 0.099 0.7>
      Morning Color=<rgba 1 0.2587 0.099 0.7>
      Day Color=<rgba 1 0.2587 0.099 0.7>
      Evening Color=<rgba 1 0.2587 0.099 0.7>
     )
     3=&FNTimeOfDayColor(
      Night Color=<rgba 1.5 0.1686 0.0312 0.3>
      Morning Color=<rgba 1.5 0.1686 0.0312 0.7>
      Day Color=<rgba 1.5 0.1686 0.0312 0.7>
      Evening Color=<rgba 1.5 0.1686 0.0312 0.7>
     )
     Index=&SpawnGroup()
     Num=::NewEnumerator2
    )
    Lifetime=13
    Mass Max=1.5
    Mass Mode=::NewEnumerator3
    Position Offset=&MergeHelper_VectorSwitch(0=<0 -15 1>, 1=<0 -10 1>, 2=<0 -7 1>, 3=<0 -3 1>, Index=&SpawnGroup(), Num=::NewEnumerator2)
    Sprite Rotation Mode=::NewEnumerator2
    Sprite Size Mode=::NewEnumerator3
    Uniform Sprite Size=&MergeHelper_FloatSwitch(0=10, 1=10, 2=10, 3=10, Index=&SpawnGroup(), Num=::NewEnumerator2)
    UsePositionOffset=true
    Write Material Random=true
   )
   &CameraOffset(Camera Offset Amount=1)
  ]
  particleUpdateScript [
   &ParticleState()
   &ScaleColor(
    Scale Alpha=&Multiply_Float(
     A=&Waveform(
      Global Input Position=&MergeHelper_FloatSwitch(0=0.35, 1=0.25, 2=0.15, 3=0.05, Index=&SpawnGroup(), Num=::NewEnumerator2)
      "Waveform [1]"=::NewEnumerator4
      "[1] Input Position"=-0.7296
      "[1] Frequency"=$User.BlinkFreq
      "[1] Amplitude Min/Max"=<0 1>
      "[1] Scale"=&FloatFromCurve(
       FloatCurve=DataInterface() {propertyValues {Curve {keys [<at 1 1>,<at 7.9587 1>,<at 8.2407 0.6>,<at 8.3993 1.2>]}}}
       CurveIndex=$Particles.Age
      )
      "[1] Bias"=&FloatFromCurve(FloatCurve=DataInterface() {propertyValues {Curve {keys [<at 0 0 RCIM_Constant>,<at 8.3993 3>,<at 9.6274 2>]}}}, CurveIndex=$Particles.Age)
     )
     B=&Waveform(
      "Waveform [1]"=::NewEnumerator0
      "[1] Input Position"=-0.7296
      "[1] Frequency"=&FloatFromCurve(FloatCurve=DataInterface() {propertyValues {Curve {keys [<at 0 0 RCIM_Constant>,<at 8.3993 1>]}}}, CurveIndex=$Particles.Age)
      "[1] Amplitude Min/Max"=<0 1>
      "[1] Bias"=&FloatFromCurve(FloatCurve=DataInterface() {propertyValues {Curve {keys [<at 0 1 RCIM_Constant>,<at 8.3993 0>]}}}, CurveIndex=$Particles.Age)
     )
    )
    ScaleA=true
    ScaleRGB=true
   )
  ]
  rendererClasses [&NiagaraSpriteRendererProperties]
  renderers [
   {rendererClass=&NiagaraSpriteRendererProperties}
   {rendererClass=&NiagaraSpriteRendererProperties, rendererIndex=1}
   {rendererClass=&NiagaraSpriteRendererProperties, rendererIndex=2}
   {rendererClass=&NiagaraSpriteRendererProperties, rendererIndex=3}
  ]
 }
 MainLight {
  emitterUpdateScript [
   &EmitterState(Life Cycle Mode=::NewEnumerator1, Loop Behavior=::NewEnumerator1, UseLoopDelay=false)
   &SpawnBurst_Instantaneous(Use Spawn Probability=false, UseLoopCountLimit=false)
   &SpawnBurst_Instantaneous(Spawn Group=1, Spawn Time=8.7, Use Spawn Probability=false, UseLoopCountLimit=false) {module=SpawnBurst_Instantaneous001}
  ]
  particleSpawnScript [
   /SetVariables(Particles.VisibilityTag=&SpawnGroup()) {isSetParameters=true}
   &InitializeParticle(
    Color=&MergeHelper_ColorSwitch(
     0=<rgba 2 0.2064 0.0534 0.5>
     1=<rgba 5 0.5159 0.1335 0.7739>
     2=?
     3=?
     Index=&SpawnGroup()
     Num=::NewEnumerator0
    )
    Lifetime=&MergeHelper_FloatSwitch(0=13, 1=0.4, 2=?, 3=?, Index=&SpawnGroup(), Num=::NewEnumerator0)
    Mass Max=1.5
    Mass Mode=::NewEnumerator3
    Mesh Scale=<1 1 0.1211>
    Mesh Scale Mode=::NewEnumerator4
    Sprite Rotation Mode=::NewEnumerator2
    Sprite Size Mode=::NewEnumerator3
    Uniform Sprite Size=20
    UsePositionOffset=false
    Write Material Random=true
   )
   &CameraOffset(Camera Offset Amount=10)
   &InitialMeshOrientation(Mesh Orientation Mode=::NewEnumerator6, Orientation Quaternion:Quat4f=<-1 0 0 1>, Use Rotation Vector=false)
  ]
  particleUpdateScript [
   &ParticleState()
   &ScaleColor(
    Scale Alpha=&MergeHelper_FloatSwitch(
     0=&Multiply_Float(
      A=&Waveform(
       "Waveform [1]"=::NewEnumerator4
       "[1] Input Position"=-0.7296
       "[1] Frequency"=$User.BlinkFreq
       "[1] Amplitude Min/Max"=<0 1>
       "[1] Scale"=&FloatFromCurve(
        FloatCurve=DataInterface() {propertyValues {Curve {keys [<at 1 1>,<at 7.9587 1>,<at 8.2407 0.6>,<at 8.3993 1.7>,<at 8.4676 1.2>]}}}
        CurveIndex=$Particles.Age
       )
       "[1] Bias"=&FloatFromCurve(FloatCurve=DataInterface() {propertyValues {Curve {keys [<at 0 0 RCIM_Constant>,<at 8.3993 3>,<at 9.6274 2>]}}}, CurveIndex=$Particles.Age)
      )
      B=&Waveform(
       "Waveform [1]"=::NewEnumerator0
       "[1] Input Position"=-0.7296
       "[1] Frequency"=&FloatFromCurve(FloatCurve=DataInterface() {propertyValues {Curve {keys [<at 0 0 RCIM_Constant>,<at 8.3993 1>]}}}, CurveIndex=$Particles.Age)
       "[1] Amplitude Min/Max"=<0 1>
       "[1] Bias"=&FloatFromCurve(FloatCurve=DataInterface() {propertyValues {Curve {keys [<at 0 1 RCIM_Constant>,<at 8.3993 0>]}}}, CurveIndex=$Particles.Age)
      )
     )
     1=&FloatFromCurve(FloatCurve=DataInterface() {propertyValues {Curve {keys [<at 0 1>,<at 1 0>]}}})
     2=?
     3=?
     Index=&SpawnGroup()
     Num=::NewEnumerator0
    )
    Scale RGB=&Multiply_Vector(
     A=<1 1 1>
     B=&VectorFromFloat(Value=&FloatFromCurve(FloatCurve=DataInterface() {propertyValues {Curve {keys [<at 0 0.8 RCIM_Constant>,<at 8.3993 2>]}}}, CurveIndex=$Particles.Age))
    )
    ScaleA=true
    ScaleRGB=true
   )
   &FN_Light_Attributes(
    Color=<rgba 1 0.2226 0.1562 1>
    Create Color Attribute=true
    Create Radius Attribute=true
    Radius=40
    Scale Color=&Multiply_Float(
     A=&Waveform(
      "Waveform [1]"=::NewEnumerator4
      "[1] Input Position"=-0.7296
      "[1] Frequency"=$User.BlinkFreq
      "[1] Amplitude Min/Max"=<0 1>
      "[1] Scale"=&FloatFromCurve(
       FloatCurve=DataInterface() {propertyValues {Curve {keys [<at 1 1>,<at 7.9587 1>,<at 8.2407 0.6>,<at 8.3993 3>,<at 8.4676 1.5>]}}}
       CurveIndex=$Particles.Age
      )
      "[1] Bias"=&FloatFromCurve(FloatCurve=DataInterface() {propertyValues {Curve {keys [<at 0 0 RCIM_Constant>,<at 8.3993 3>,<at 9.6274 2>]}}}, CurveIndex=$Particles.Age)
     )
     B=&Waveform(
      "Waveform [1]"=::NewEnumerator0
      "[1] Input Position"=-0.7296
      "[1] Frequency"=&FloatFromCurve(FloatCurve=DataInterface() {propertyValues {Curve {keys [<at 0 0 RCIM_Constant>,<at 8.3993 1>]}}}, CurveIndex=$Particles.Age)
      "[1] Amplitude Min/Max"=<0 1>
      "[1] Bias"=&FloatFromCurve(FloatCurve=DataInterface() {propertyValues {Curve {keys [<at 0 1 RCIM_Constant>,<at 8.3993 0>]}}}, CurveIndex=$Particles.Age)
     )
    )
   )
   /SetVariables(Particles.Linear Color=&MakeLinearColorFromVectorAndFloat(
     "Vector (RGB)"=&MakeVectorFromLinearColorRGB(Color=<rgba 1 0.1032 0.0267 1>)
     "Float (Alpha)"=&Multiply_Float(
      A=0.05
      B=&Multiply_Float(
       A=&Waveform(
        "Waveform [1]"=::NewEnumerator4
        "[1] Input Position"=-0.7296
        "[1] Frequency"=$User.BlinkFreq
        "[1] Amplitude Min/Max"=<0 1>
        "[1] Scale"=&FloatFromCurve(
         FloatCurve=DataInterface() {propertyValues {Curve {keys [<at 1 1>,<at 7.9587 1>,<at 8.2407 0.6>,<at 8.3993 3>,<at 8.4676 1.5>]}}}
         CurveIndex=$Particles.Age
        )
        "[1] Bias"=&FloatFromCurve(FloatCurve=DataInterface() {propertyValues {Curve {keys [<at 0 0 RCIM_Constant>,<at 8.3993 3>,<at 9.6274 2>]}}}, CurveIndex=$Particles.Age)
       )
       B=&Waveform(
        "Waveform [1]"=::NewEnumerator0
        "[1] Input Position"=-0.7296
        "[1] Frequency"=&FloatFromCurve(FloatCurve=DataInterface() {propertyValues {Curve {keys [<at 0 0 RCIM_Constant>,<at 8.3993 1>]}}}, CurveIndex=$Particles.Age)
        "[1] Amplitude Min/Max"=<0 1>
        "[1] Bias"=&FloatFromCurve(FloatCurve=DataInterface() {propertyValues {Curve {keys [<at 0 1 RCIM_Constant>,<at 8.3993 0>]}}}, CurveIndex=$Particles.Age)
       )
      )
     )
    )) {isSetParameters=true}
  ]
  rendererClasses [
   &NiagaraLightRendererProperties
   &NiagaraMeshRendererProperties
   &NiagaraSpriteRendererProperties
  ]
  renderers [
   {rendererClass=&NiagaraLightRendererProperties, rendererIndex=0}
   {rendererClass=&NiagaraMeshRendererProperties, rendererIndex=1}
   {rendererClass=&NiagaraSpriteRendererProperties, rendererIndex=2}
  ]
 }
 WingLights {
  emitterUpdateScript [
   &EmitterState(Life Cycle Mode=::NewEnumerator1, Loop Behavior=::NewEnumerator1, Loop Duration=13, UseLoopDelay=false)
   &SpawnBurst_Instantaneous(Spawn Count=2, Use Spawn Probability=false, UseLoopCountLimit=false)
   &SpawnBurst_Instantaneous(Spawn Count=2, Spawn Group=1, Spawn Time=8.7, Use Spawn Probability=false, UseLoopCountLimit=false) {module=SpawnBurst_Instantaneous001}
  ]
  particleSpawnScript [
   &PartitionParticles()
   &InitializeParticle(
    Color=&MergeHelper_ColorSwitch(
     0=<rgba 2 0.2064 0.0534 0.5>
     1=<rgba 5 0.5159 0.1335 0.7739>
     2=?
     3=?
     Index=&SpawnGroup()
     Num=::NewEnumerator0
    )
    Lifetime=&MergeHelper_FloatSwitch(0=13, 1=0.4, 2=?, 3=?, Index=&SpawnGroup(), Num=::NewEnumerator0)
    Mass Max=1.5
    Mass Mode=::NewEnumerator3
    Sprite Rotation Mode=::NewEnumerator0
    Sprite Size=&MergeHelper_Vector2Switch(0=<10 21>, 1=<13 21>, 2=?, 3=?, 4=?, Index=&SpawnGroup(), Num=::NewEnumerator0)
    Sprite Size Mode=::NewEnumerator4
    UsePositionOffset=false
    Write Material Random=true
   )
   &CameraOffset(Camera Offset Amount=2)
   &ApplyInitialForces()
  ]
  particleUpdateScript [
   &ParticleState()
   &SolveForcesAndVelocity(Clamp Velocity=false, Limit Acceleration=false)
   &SkeletalMeshLocation(
    Bone / Socket Index=&SelectIntFromArray(Direct Array Index=$Particles.PartitionParticles.Partition, Int Selection Array=DataInterface() {propertyValues {IntData [1,0]}})
    MaskSpawnGroup=false
    Mesh Sampling Type=::NewEnumerator5
    Skeletal Mesh=$System.Skeletal Mesh
    Socket Sampling Mode=::NewEnumerator5
   )
   /SetVariables(Particles.Vector:Vector3f=&RotateVectorByQuaternion(
     Vector To Rotate=&SelectVectorFromArray(
      Direct Array Index=$Particles.PartitionParticles.Partition
      Vector Selection Array=DataInterface() {propertyValues {FloatData [<0 1 0>, <0 1 0>]}}
     )
     Quaternion=$Particles.SkeletalMeshLocation.SampledRotation
    )) {isSetParameters=true}
   &ScaleColor(
    Scale Alpha=&MergeHelper_FloatSwitch(
     0=&Multiply_Float(
      A=&Waveform(
       "Waveform [1]"=::NewEnumerator4
       "[1] Input Position"=-0.7296
       "[1] Frequency"=$User.BlinkFreq
       "[1] Amplitude Min/Max"=<0 1>
       "[1] Scale"=&FloatFromCurve(
        FloatCurve=DataInterface() {propertyValues {Curve {keys [<at 1 1>,<at 7.4106 1>,<at 8.2802 0.6>,<at 8.3 5>,<at 8.4676 1.3>]}}}
        CurveIndex=$Particles.Age
       )
       "[1] Bias"=&FloatFromCurve(FloatCurve=DataInterface() {propertyValues {Curve {keys [<at 0 0 RCIM_Constant>,<at 8.4 3>,<at 9.6274 2>]}}}, CurveIndex=$Particles.Age)
      )
      B=&Waveform(
       "Waveform [1]"=::NewEnumerator0
       "[1] Input Position"=-0.7296
       "[1] Frequency"=&FloatFromCurve(FloatCurve=DataInterface() {propertyValues {Curve {keys [<at 0 0 RCIM_Constant>,<at 8.3993 1>]}}}, CurveIndex=$Particles.Age)
       "[1] Amplitude Min/Max"=<0 1>
       "[1] Bias"=&FloatFromCurve(FloatCurve=DataInterface() {propertyValues {Curve {keys [<at 0 1 RCIM_Constant>,<at 8.3993 0>]}}}, CurveIndex=$Particles.Age)
      )
     )
     1=&FloatFromCurve(FloatCurve=DataInterface() {propertyValues {Curve {keys [<at 0 1>,<at 1 0>]}}})
     2=?
     3=?
     Index=&SpawnGroup()
     Num=::NewEnumerator0
    )
    ScaleA=true
    ScaleRGB=true
   )
  ]
  rendererClasses [&NiagaraSpriteRendererProperties]
  renderers [{rendererClass=&NiagaraSpriteRendererProperties}]
 }
}
summary {
 systemName=NS_Emote_PineTrim_Crisp_LightFX
 emitters [
  {
   emitterName=MainLight
   rendererClasses [
    &NiagaraLightRendererProperties
    &NiagaraMeshRendererProperties
    &NiagaraSpriteRendererProperties
   ]
  }
  {emitterName=WingLights, rendererClasses [&NiagaraSpriteRendererProperties]}
  {emitterName=BodyLights, rendererClasses [&NiagaraSpriteRendererProperties]}
 ]
 userVariables [User.BlinkFreq {type=NiagaraFloat, defaultValue=1, description=""}]
}
